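# Unlit translucent material for Niagara mesh particles that draws the animated music effect of the Lyrical female backpack.
# A panned, axis-swapped texture drives opacity, and its green channel can index a color curve atlas before particle tint and optional desaturation.
# GlowMESH blends this toward a plain particle-colored glow shaped by a sine falloff across U.
# Two time-driven sine waves along V add a wobble offset, sine-masked along V and scaled by a dynamic WPO parameter.
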Material /Game/Accessories/FORT_Backpacks/M_LyricalFemale/Material/M_Backpack_Lycial_Music
graph {
 properties {
  bUsedWithNiagaraMeshParticles=true
  blendMode=BLEND_Translucent
  pixelDepthOffsetMode=PDOM_Legacy
  shadingModel=MSM_Unlit
  usedShadingModels=Unlit
 }
 nodes [
  n1 = ParticleColor()
  n2 = VectorParameter() {defaultValue=<rgba 1 3 0 1>}
  n3 = Multiply(A=TextureCoordinate(), B=AppendVector(A=@n2.R, B=@n2.R))
  n4 = TextureSampleParameter2D(UVs=Panner(Coordinate=MaterialFunctionCall(XY=@n3) {materialFunction=&Swizzle}.YX) {speedX=2}) {parameterName=Texture, texture=&/T_Backpack_Lyrical}
  n5 = ParticleColor()
  n6 = TextureCoordinate()
  n7 = Sine(None=Multiply(A=ComponentMask(None=@n6) {r=true}) {constB=0.5})
  n8 = Constant() {r=4}
  LinearInterpolate(A=Multiply(A=@n1.A, B=@n4.RGBA), B=Multiply(A=@n5.A, B=Power(Base=@n7, Exp=@n8)), Alpha=ScalarParameter() {parameterName=GlowMESH})
  n9 = CurveAtlasRowParameter(CurveTime=@n4.G) {
   atlas=/Game/Athena/LinearColorCurves/A_Atlas_Turtleneck_Glider
   curve=&/Curve_Backpack_Lyrical_Color
   defaultValue=8
   parameterName=Param_color
   sortPriority=0
  }
  n10 = Multiply(A=LinearInterpolate(A=@n9, B=@n4.G, Alpha=ScalarParameter() {parameterName=Use_Curve}), B=@n1.RGB)
  n11 = ScalarParameter() {parameterName=Desaturation}
  MaterialFunctionCall(Emissive=LinearInterpolate(A=Desaturation(None=@n10, Fraction=@n11), B=@n5.RGB, Alpha=ScalarParameter() {parameterName=GlowMESH})) {materialFunction=&MF_FX_EmissiveAdjustments}
  Multiply()
  n12 = ComponentMask(None=TextureCoordinate()) {g=true}
  n13 = Multiply(A=ComponentMask(None=TextureCoordinate()) {g=true}, B=ScalarParameter() {defaultValue=10, parameterName=WPO_UV_A}) {constB=0.5}
  n14 = Multiply(A=Time()) {constB=3}
  n15 = LinearInterpolate(Alpha=Sine(None=Add(A=@n13, B=@n14))) {constA=-1}
  n16 = Constant()
  n17 = Multiply(
   A=MaterialFunctionCall(X=@n15, Y=Constant(), Z=Multiply(A=@n15, B=@n16)) {materialFunction=&MakeFloat3}.Result
   B=Multiply(A=ScalarParameter() {defaultValue=0.9, parameterName=WPO}, B=DynamicParameter() {paramNames [WPO,Param2,Param3,Param4]}.WPO)
  )
  n18 = Multiply(A=ComponentMask(None=TextureCoordinate()) {g=true}, B=ScalarParameter() {defaultValue=4, parameterName=WPO_UV_B}) {constB=0.5}
  n19 = Multiply(A=OneMinus(None=Time())) {constB=5}
  n20 = LinearInterpolate(Alpha=Sine(None=Add(A=@n18, B=@n19))) {constA=-1}
  n21 = Constant()
  n22 = Multiply(A=@n20, B=Constant())
  n23 = ScalarParameter() {defaultValue=0.9, parameterName=WPO}
  n24 = DynamicParameter() {paramNames [WPO,Param2,Param3,Param4]}
  n25 = Multiply(
   A=Multiply(
    A=MaterialFunctionCall(X=@n20, Y=@n21, Z=@n22) {materialFunction=&MakeFloat3}.Result
    B=Multiply(A=@n23, B=@n24.WPO)
   )
   B=Constant() {r=-1}
  )
  Multiply(A=Sine(None=Multiply(A=@n12) {constB=0.5}), B=Transform(None=Add(A=@n17, B=@n25)))
 ]
}
