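# Unlit, two-sided translucent material for the Lyrical backpack's effects, usable on Niagara mesh particles.
# World position offset combines sine spikes, panning smoke noise and SplineThicken width, jittered by stepped-time noise.
# Scalar lerp switches select vertex-color, dark-line, glow or pickaxe behavior, the last zeroing the offset.
# Emissive and opacity combine particle color and alpha with time-of-day values, plus a UV sine falloff for glow.
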Material /Game/Accessories/FORT_Backpacks/M_LyricalFemale/Material/M_Backpack_Lyrical
graph {
 properties {
  allowTranslucentCustomDepthWrites=true
  bEnableResponsiveAA=true
  bUsedWithNiagaraMeshParticles=true
  blendMode=BLEND_Translucent
  opacityMaskClipValue=0.05
  pixelDepthOffsetMode=PDOM_Legacy
  shadingModel=MSM_Unlit
  twoSided=true
  usedShadingModels=Unlit
 }
 nodes [
  n1 = TextureCoordinate() {vTiling=16}
  n2 = Add(A=Multiply(A=ComponentMask(None=@n1) {g=true}) {constB=0.5}, B=Time())
  n3 = TextureCoordinate() {vTiling=4}
  n4 = Time()
  n5 = Add(A=Multiply(A=ComponentMask(None=@n3) {g=true}) {constB=0.5}, B=Multiply(A=OneMinus(None=@n4)) {constB=0.55})
  n6 = Add(A=Power(Base=Sine(None=@n2)), B=Multiply(A=Sine(None=@n5)) {constB=0.4})
  n7 = Power(Base=LinearInterpolate(Alpha=Clamp(None=@n6)) {constA=-1})
  n8 = DynamicParameter() {paramNames [spikes offset, Param2, Param3, Param4]}
  n9 = Multiply(
   A=MaterialFunctionCall(X=@n7, Y=@n7, Z=@n7) {materialFunction=&MakeFloat3}.Result
   B=@n8.spikes offset
  )
  n10 = VectorParameter() {parameterName=WPO_UV, defaultValue=<rgba 0 0 0 3>}
  n11 = Panner(Coordinate=Multiply(A=TextureCoordinate(), B=AppendVector(A=@n10.R, B=@n10.R)), Time=Time()) {speedX=1, speedY=1}
  n12 = MaterialFunctionCall() {materialFunction=&DebugTimeSine}
  n13 = Multiply(
   A=Power(
    Base=TextureSample(UVs=@n11) {texture=&T_Inky_Smoke_Tile}.RGB
    Exp=ScalarParameter() {defaultValue=2, group=Main_WPO, parameterName=WPO_Pwe}
   )
   B=Multiply(A=ScalarParameter() {defaultValue=1.8, group=Main_WPO, parameterName=WPO}, B=LinearInterpolate(Alpha=@n12.Result) {constA=0.4})
  )
  n14 = WorldPosition()
  n15 = ObjectPositionWS()
  n16 = Panner(Coordinate=TextureCoordinate() {vTiling=5}) {speedX=2.2}
  n17 = Multiply(
   A=Normalize(VectorInput=Subtract(A=@n14.XYZ, B=@n15))
   B=Multiply(
    A=TextureSample(UVs=@n16) {texture=&T_Inky_Smoke_Tile}.RGB
    B=ScalarParameter() {defaultValue=2.5, parameterName=Param_1}
   )
  )
  n18 = VertexColor()
  n19 = ScalarParameter() {defaultValue=0.42, group=Main_WPO, parameterName=SplineThicken_NoiseSizeA}
  n20 = ScalarParameter() {defaultValue=1.8, group=Main_WPO, parameterName=SplineThicken_NoiseSizeB}
  n21 = VectorParameter() {defaultValue=<rgba 0 0 0.5 4>}
  n22 = AppendVector(A=@n21.B, B=@n21.B)
  n23 = Round(None=Multiply(A=Time(), B=ScalarParameter() {defaultValue=12, group=Main_WPO, parameterName=SplineSpeed}))
  n24 = Panner(Coordinate=Multiply(A=TextureCoordinate() {uTiling=0, vTiling=1.1}, B=@n22), Time=@n23) {bFractionalPart=true, speedX=0.1, speedY=-0.55}
  n25 = Panner(Coordinate=Multiply(A=TextureCoordinate() {vTiling=0.3}, B=@n22), Time=@n23) {bFractionalPart=true, speedX=0.1, speedY=0.2}
  n26 = LinearInterpolate(
   A=LinearInterpolate(Alpha=TextureSample(UVs=@n24) {samplerType=SAMPLERTYPE_Grayscale, texture=&LowResBlurredNoise}.R) {constA=-1}
   B=Power(Base=TextureSample(UVs=@n25) {samplerType=SAMPLERTYPE_Normal, texture=/Game/Global/Textures/Default/Tileable_Patterns/2ndInfectedMapNormal}.R) {constExponent=1}
  )
  n27 = LinearInterpolate(
   A=Multiply(
    A=ScalarParameter() {defaultValue=1, group=Main_WPO, parameterName=SplineThicken_Noise}
    B=LinearInterpolate(A=@n19, B=@n20, Alpha=@n26) {constA=0.42, constB=1.8}
   )
   B=Constant() {r=1}
   Alpha=ScalarParameter() {group=Switch, parameterName=SplineThicken_Noise_Switch}
  )
  n28 = ScalarParameter() {defaultValue=2.2, group=Main_WPO, parameterName=SplineThicken}
  n29 = Multiply(A=@n27, B=@n28)
  n30 = Multiply(A=@n27, B=@n28)
  n31 = Add(
   A=LinearInterpolate(A=Add(A=@n9, B=@n13), B=Multiply(A=@n17, B=@n18.R), Alpha=ScalarParameter() {group=Switch, parameterName=Use_VertexColor})
   B=LinearInterpolate(
    A=MaterialFunctionCall(WidthBase=@n29, WidthTip=@n30) {materialFunction=&SplineThicken}.WorldPositionOffset
    B=Constant()
    Alpha=ScalarParameter() {group=Switch, parameterName=Switch_SplineThicken}
   )
  )
  n32 = TextureCoordinate()
  n33 = VectorParameter() {parameterName=UV_Dark, defaultValue=<rgba 0 5 0.5 4>}
  n34 = AppendVector(A=@n33.R, B=@n33.R)
  n35 = TextureSample(UVs=Panner(Coordinate=Multiply(A=@n32, B=@n34), Time=Time()) {speedX=2, speedY=1}) {samplerType=SAMPLERTYPE_Grayscale, texture=&LowResBlurredNoise}
  n36 = VectorParameter() {parameterName=Param_1, defaultValue=<rgba 0.8 0.2 0.3 1>}
  n37 = ScalarParameter() {defaultValue=5.5, group=DarkLine, parameterName=WPO_Dark}
  n38 = LinearInterpolate(Alpha=MaterialFunctionCall() {materialFunction=&DebugTimeSine}.Result) {constA=0.55}
  n39 = Multiply(
   A=Power(Base=Multiply(A=@n35.R, B=@n36.RGB), Exp=Constant() {r=2})
   B=Multiply(A=Multiply(A=@n37, B=@n38), B=DynamicParameter() {paramNames [spikes offset, Param2, Param3, Param4]}.Param2)
  )
  n40 = ScalarParameter() {defaultValue=0.52, group=DarkLine, parameterName=SplineThicken_Dark}
  n41 = Add(
   A=@n39
   B=MaterialFunctionCall(WidthBase=@n40, WidthTip=@n40) {materialFunction=&SplineThicken}.WorldPositionOffset
  )
  n42 = ScalarParameter() {group=Switch, parameterName=Switch_Dark}
  n43 = ScalarParameter() {defaultValue=1, parameterName=GlowSplineThicken_Multipliy}
  n44 = ScalarParameter() {defaultValue=6, group=Glow, parameterName=SplineThicken_Glow}
  n45 = MaterialFunctionCall(WidthBase=@n44, WidthTip=@n44) {materialFunction=&SplineThicken}
  LinearInterpolate(
   A=LinearInterpolate(
    A=LinearInterpolate(A=@n31, B=@n41, Alpha=@n42)
    B=Multiply(A=@n43, B=@n45.WorldPositionOffset)
    Alpha=ScalarParameter() {group=Switch, parameterName=Switch_Glow}
   )
   B=Constant()
   Alpha=ScalarParameter() {group=Switch, parameterName=Use_Pickaxe}
  )
  n46 = ParticleColor()
  n47 = MaterialFunctionCall(
   Color_Morning=ScalarParameter() {defaultValue=1, parameterName=Color_Morning}
   Color_Day=ScalarParameter() {defaultValue=1, parameterName=Color_Day}
   Color_Evening=ScalarParameter() {defaultValue=1, parameterName=Color_Evening}
   Color_Night=ScalarParameter() {defaultValue=0.65, parameterName=Color_Night}
  ) {materialFunction=&Color_Time_of_Day}
  n48 = Multiply(A=ComponentMask(None=TextureCoordinate()) {r=true}) {constB=0.5}
  n49 = Multiply(A=@n46.A, B=Power(Base=Sine(None=@n48), Exp=ScalarParameter() {defaultValue=2, parameterName=Glow_Power}))
  n50 = VertexColor()
  LinearInterpolate(
   A=LinearInterpolate(A=Multiply(A=@n46.A, B=@n47.Result), B=Multiply(A=@n49, B=@n50.G), Alpha=ScalarParameter() {group=Switch, parameterName=Switch_Glow})
   B=@n46.A
   Alpha=ScalarParameter() {group=Switch, parameterName=Use_Pickaxe}
  )
  n51 = ScalarParameter() {defaultValue=10, parameterName=EmissiveMultiply}
  n52 = DynamicParameter() {paramNames [spikes offset, Param2, Param3, Param4]}
  n53 = Multiply(
   A=LinearInterpolate(A=@n51, B=Multiply(A=@n52.Param3, B=@n51), Alpha=ScalarParameter() {group=Switch, parameterName=Use_Pickaxe})
   B=MaterialFunctionCall(
    Color_Morning=VectorParameter() {parameterName=EmissiveColor_Morning, defaultValue=<rgba 0.6667 0.0241 0 1>}.RGB
    Color_Day=VectorParameter() {parameterName=EmissiveColor_Day, defaultValue=<rgba 1 0.1448 0 1>}.RGB
    Color_Evening=VectorParameter() {parameterName=EmissiveColor_Evening, defaultValue=<rgba 0.934 0.1352 0 1>}.RGB
    Color_Night=VectorParameter() {parameterName=EmissiveColor_Night, defaultValue=<rgba 1 0.4457 0 1>}.RGB
   ) {materialFunction=&Color_Time_of_Day}.Result
  )
  MaterialFunctionCall(Emissive=LinearInterpolate(A=Multiply(A=@n53, B=@n46.RGB), B=@n46.RGB, Alpha=ScalarParameter() {group=Switch, parameterName=Switch_Glow})) {materialFunction=&MF_FX_EmissiveAdjustments}
  Multiply(A=ScalarParameter() {defaultValue=1}, B=@n8.spikes offset)
  Transform(None=@n13) {transformType=TRANSFORM_Local}
  Transform(None=@n39)
  Transform(None=@n9)
 ]
}
